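; Blueprint for the Manor variant of an ornate dome light prop in the Helios environment, apparently built on a parent class with time-of-day support.
; Its GetTimeOfDayBlueprintDefaultVariables override only returns the parent's result.
; BeginPlay, overlap, Tick, the construction script and a leftover NewFunction_0 are empty, so the graphs here add no logic of their own.
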(event EventBeginPlay)

(event Collision|EventActorBeginOverlap
  (params (OtherActor :type Actor%20Object%20Reference)))

(event EventTick
  (params (DeltaSeconds :type Float%20%28single-precision%29)))

(fn GetTimeOfDayBlueprintDefaultVariables
  (bind _outvariables (|CallParentFunction))
  (return _outvariables))

(fn NewFunction_0)

(fn ConstructionScript)
